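# Looping Niagara effect for the Snippet emote gadget, combining blade mesh flashes and a trimmer ribbon trail on a skeletal mesh.
# SharpBlade fires two staggered bursts every 4–5 seconds, using partition and spawn-group switches for bone, offset, 180-degree flip and material phase.
# Trimmer_Ribbon follows a fixed bone, spawning by owner speed and fading out with time-of-day tint.
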NiagaraSystem /BRCosmetics/Animation/Game/MainPlayer/Emotes/Snippet/FX/NS_Emote_Snipped_GadgetLoop
dependencies {
 usedDataInterfaces [&NiagaraDataInterfaceCurve]
 usedDynamicInputs []
 usedModules [
  &self:SystemSpawnScript.SystemScriptSource.SystemScriptGraph.NiagaraNodeAssignment_0.SetVariables_#1
  &SystemState
 ]
}
emitters {
 SharpBlade {
  emitterUpdateScript [
   &EmitterState(
    Life Cycle Mode=::NewEnumerator1
    Loop Duration=&RandomRangeFloat(Minimum=4, Maximum=5)
    Recalculate Duration Each Loop=true
    UseLoopDelay=false
   )
   &SpawnBurst_Instantaneous(Spawn Count=2, Use Spawn Probability=false, UseLoopCountLimit=false)
   &SpawnBurst_Instantaneous(Spawn Group=1, Spawn Time=0.25, Use Spawn Probability=false, UseLoopCountLimit=false) {module=SpawnBurst_Instantaneous001}
  ]
  particleSpawnScript [
   &InitializeParticle(
    Color=<rgba 1 1 1 2>
    Lifetime=0.6
    Mesh Scale Mode=::NewEnumerator3
    Mesh Uniform Scale=0.72
    Sprite Size Mode=::NewEnumerator3
    UsePositionOffset=false
    Write Material Random=true
   )
   &PartitionParticles()
   /SetVariables(Particles.SpawnGroup=$Emitter.SpawnGroup) {isSetParameters=true}
   /SetVariables(Particles.VisibilityTag=$Emitter.SpawnGroup) {isSetParameters=true}
  ]
  particleUpdateScript [
   &ParticleState()
   &SolveForcesAndVelocity(Clamp Velocity=false, Limit Acceleration=false)
   &SkeletalMeshLocation(
    Bone / Socket Index=&MergeHelper_IntSwitch(0=$Particles.PartitionParticles.Partition, 1=0, 2=?, 3=?, 4=?, 5=?, Index=$Particles.SpawnGroup, Num=::NewEnumerator0)
    Bone Sampling Mode=::NewEnumerator4
    MaskSpawnGroup=false
    Sampled Position Offset=&MergeHelper_VectorSwitch(0=<0 0 0>, 1=<0 0 14>, 2=?, 3=?, Index=$Particles.SpawnGroup, Num=::NewEnumerator0)
    Skeletal Mesh=$System.Skeletal Mesh
   )
   /SetVariables(Particles.MeshOrientation=/Game/Effects/Niagara/Modules/MergeEmitters/MergeHelper_QuatSwitch(
     0:Quat4f=$Particles.SkeletalMeshLocation.SampledRotation
     1:Quat4f=&MultiplyQuaternion(
      Quaternion A=$Particles.SkeletalMeshLocation.SampledRotation
      Quaternion B=&MakeQuatFromAxisAngle(Axis=<0 0 1>, AngleInDegrees=180)
     )
     2:Quat4f=?
     3:Quat4f=?
     4:Quat4f=?
     5:Quat4f=?
     6:Quat4f=?
     7:Quat4f=?
     8:Quat4f=?
     9:Quat4f=?
     Index:NiagaraInt32=$Particles.PartitionParticles.Partition
     Num:MergeHelper_SwitcherEnum=::NewEnumerator0
    )) {isSetParameters=true}
   &SpriteRotationRate(Rotation Rate=&RandomRangeFloat(Minimum=150, Maximum=180))
   &ScaleSpriteSize(
    Uniform Curve Scale=2
    Uniform Curve Sprite Scale=DataInterface() {propertyValues {Curve {keys [
        {arriveTangent=9.0725, interpMode=RCIM_Cubic, leaveTangent=9.0725, tangentMode=RCTM_User, time=0, value=0}
        {arriveTangent=-1.7917, interpMode=RCIM_Cubic, leaveTangent=-1.7917, tangentMode=RCTM_User, time=0.0666, value=0.9847}
        {arriveTangent=-0.0405, interpMode=RCIM_Cubic, leaveTangent=-0.0405, tangentMode=RCTM_User, time=1, value=0}
       ]}}}
   )
   &CameraOffset(Camera Offset Amount=1)
   &DynamicMaterialParameters(
    Index 0 Param 1=&ModuloFloat(
     A=&Add_Float(
      A=$Particles.NormalizedAge
      B=&MergeHelper_FloatSwitch(0=0, 1=0.06, 2=?, 3=?, Index=$Particles.PartitionParticles.Partition, Num=::NewEnumerator0)
     )
     B=3
    )
    Index 0 Param 2=1
    Module.Index0Param0Write=true
    Module.Index0Param1Write=true
    Module.Index0Param2Write=true
    Module.Index0Param3Write=true
   )
  ]
  rendererClasses [&NiagaraMeshRendererProperties]
  renderers [{rendererClass=&NiagaraMeshRendererProperties, rendererIndex=0}]
 }
 Trimmer_Ribbon {
  emitterUpdateScript [
   &EmitterState()
   &FN_SpawnRibbon_v2(
    Spawn Mode=::NewEnumerator0
    SpawnRate:NiagaraFloat=&FloatFromCurve(
     FloatCurve=DataInterface() {propertyValues {Curve {keys [{interpMode=RCIM_Cubic, tangentMode=RCTM_User, time=0, value=0}, {interpMode=RCIM_Cubic, tangentMode=RCTM_User, time=1, value=1}]}}}
     CurveIndex=&RemapRange(
      Input Value To Remap=&VectorLength(Vector=$Engine.Owner.Velocity)
      Input Min=70
      Input Max=100
      Output Min=0
      Output Max=1
     )
     Scale Curve=30
    )
   )
   &FN_Integrator() {enabled=false}
  ]
  particleSpawnScript [
   /Game/Effects/Niagara/Modules/Ribbons/FN_InitializeRibbon_v4(
    Color:LinearColor=<rgba 1 1 1 5>
    Enable Skeletal Position=true
    Lifetime=0.4
    Ribbon Width=8
    SkeletalMesh:NiagaraDataInterfaceSkeletalMesh=$User.SkeletalMesh
   )
   &FN_FlipRibbonV() {enabled=false}
   &AddVelocity(Velocity=<0 0 100>) {enabled=false}
   &SkeletalMeshLocation(
    Bone / Socket Index=2
    Bone Sampling Mode=::NewEnumerator4
    Interpolate Mesh Sampling=false
    MaskSpawnGroup=false
    Sampled Position Offset=<10 0 0>
    Skeletal Mesh=$System.Skeletal Mesh
   )
  ]
  particleUpdateScript [
   &InheritSourceMovement(ApplyForce=false, ApplyPosition=true, ApplyPositionFalloff=false) {enabled=false}
   &SolveForcesAndVelocity(Clamp Velocity=false, Limit Acceleration=false)
   /Game/Effects/Niagara/Modules/Ribbons/FN_UpdateRibbon_v4(Enable Skeletal Position=true, SkeletalMesh:NiagaraDataInterfaceSkeletalMesh=$User.SkeletalMesh)
   &FN_RibbonMaterial(
    Color=&MultiplyLinearColor(
     Linear Color=<rgba 0.1667 0.1667 0.1667 0.7>
     Scale Factor=&MakeVector4_FromLinearColor(LinearColor=&FN_Select_Color_from_Time_Of_Day(Morning=<rgba 0.4937 0.298 0.5 0.5>, Day=<rgba 1 1 1 0.7>, Evening=<rgba 0.5 0.5 0.5 0.5>, Night=<rgba 0.1303 0.3196 1 1>))
    )
    Scale Alpha:NiagaraFloat=&FloatFromCurve(FloatCurve=DataInterface() {propertyValues {Curve {keys [
         {arriveTangent=-3.0003, interpMode=RCIM_Cubic, leaveTangent=-3.0003, tangentMode=RCTM_User, time=0, value=1}
         {interpMode=RCIM_Cubic, tangentMode=RCTM_User, time=1, value=0}
        ]}}})
   )
  ]
  rendererClasses [&NiagaraRibbonRendererProperties]
  renderers [{rendererClass=&NiagaraRibbonRendererProperties}]
 }
}
summary {
 systemName=NS_Emote_Snipped_GadgetLoop
 emitters [
  {emitterName=Trimmer_Ribbon, rendererClasses [&NiagaraRibbonRendererProperties]}
  {emitterName=SharpBlade, rendererClasses [&NiagaraMeshRendererProperties]}
 ]
 userVariables [
  User.bisSwinging {type=NiagaraBool}
  User.RibbonIDOffset {type=NiagaraInt32}
  User.SkeletalMesh {
   type=NiagaraDataInterfaceSkeletalMesh
   defaultValue {
    struct=NiagaraExt_VariableValue_DataInterface
    value {dataInterface=&self:NiagaraDataInterfaceSkeletalMesh_0}
   }
  }
 ]
}
